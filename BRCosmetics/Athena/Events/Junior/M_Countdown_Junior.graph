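# Translucent unlit material for the Junior event's countdown display, combining timer digits, localized unit text, a progress bar and animated rings.
# UV1 ranges select mesh parts, with one panned or stepped mask texture for the rings.
# MPC_Countdown digit pairs reaching zero switch stage colors, bar fill and vertex-color-masked vertex offsets.
# Show flags, a phase-2 switch and close-camera fade hide it by zeroing opacity or collapsing and displacing vertices.
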Material /BRCosmetics/Athena/Events/Junior/M_Countdown_Junior
graph {
 properties {
  bEnableResponsiveAA=true
  bUsedWithInstancedStaticMeshes=true
  blendMode=BLEND_Translucent
  pixelDepthOffsetMode=PDOM_Legacy
  shadingModel=MSM_Unlit
  usedShadingModels=Unlit
 }
 nodes [
  Constant()
  Constant() {r=1}
  Constant()
  Constant()
  n1 = VectorParameter() {parameterName=ColorInsideStage1, defaultValue=<rgba 0 4.5991 10 1>}
  n2 = VectorParameter() {parameterName=ColorInsideStage2, defaultValue=<rgba 4 1.5466 0 1>}
  n3 = Multiply(
   A=Add(
    A=CollectionParameter() {collection=/Game/Athena/Events/Generic/CountdownTimer/Materials/MPC_Countdown, parameterName=Digit8}
    B=CollectionParameter() {collection=&MPC_Countdown, parameterName=Digit7}
   )
   B=ScalarParameter() {defaultValue=1, parameterName=DebugProgress1}
  )
  n4 = Constant() {r=1}
  n5 = If(A=@n3, B=Constant(), "A > B"=Constant(), "A == B"=@n4, "A < B"=@n4)
  n6 = CollectionParameter() {collection=&MPC_Countdown, parameterName=Digit6}
  n7 = CollectionParameter() {collection=&MPC_Countdown, parameterName=Digit5}
  n8 = Add(A=@n3, B=Multiply(A=Add(A=@n6, B=@n7), B=ScalarParameter() {defaultValue=1, parameterName=DebugProgress2}))
  n9 = Constant() {r=1}
  n10 = If(A=@n8, B=Constant(), "A > B"=Constant(), "A == B"=@n9, "A < B"=@n9)
  n11 = Subtract(A=ComponentMask(None=TextureCoordinate() {coordinateIndex=1}) {r=true}, B=Constant() {r=-1})
  n12 = Floor(None=@n11)
  n13 = Multiply(A=Saturate(None=Ceil(None=@n11)), B=OneMinus(None=Saturate(None=@n12)))
  n14 = Multiply(
   A=LinearInterpolate(
    A=LinearInterpolate(A=@n1.RGB, B=@n2.RGB, Alpha=@n5)
    B=VectorParameter() {parameterName=ColorInsideStage3, defaultValue=<rgba 10 0 0 1>}.RGB
    Alpha=@n10
   )
   B=LinearInterpolate(B=ScalarParameter() {defaultValue=0.099, parameterName=BackPlateDarken}, Alpha=@n13) {constA=1}
  )
  n15 = TextureSample() {samplerType=SAMPLERTYPE_Masks, texture=&/T_JCount_M}
  n16 = ScalarParameter() {defaultValue=0.0381, parameterName=PanRing1}
  n17 = Constant()
  n18 = TextureSample(UVs=Panner(Coordinate=TextureCoordinate(), Speed=AppendVector(A=@n16, B=@n17))) {samplerType=SAMPLERTYPE_Masks, texture=&T_JCount_M}
  n19 = LinearInterpolate(
   A=VectorParameter() {parameterName=ColorRing, defaultValue=<rgba 0 3.1632 10 1>}.RGB
   B=VectorParameter() {parameterName=ColorRing2, defaultValue=<rgba 10 0 0 1>}.RGB
   Alpha=@n10
  )
  n20 = Subtract(A=ComponentMask(None=TextureCoordinate() {coordinateIndex=1}) {r=true}, B=Constant() {r=1})
  n21 = Floor(None=@n20)
  n22 = Multiply(A=Saturate(None=Ceil(None=@n20)), B=OneMinus(None=Saturate(None=@n21)))
  n23 = LinearInterpolate(
   A=LinearInterpolate(A=Multiply(A=@n14, B=@n15.R), B=VectorParameter() {parameterName=ColorOutside, defaultValue=<rgba 2 1.3003 0 1>}.RGB, Alpha=@n15.G)
   B=Multiply(A=@n18.R, B=@n19)
   Alpha=@n22
  )
  n24 = ScalarParameter() {defaultValue=-0.0381, parameterName=PanRing2}
  n25 = Constant()
  n26 = TextureSample(UVs=Panner(Coordinate=TextureCoordinate(), Speed=AppendVector(A=@n24, B=@n25))) {samplerType=SAMPLERTYPE_Masks, texture=&T_JCount_M}
  n27 = Multiply(A=@n26.R, B=@n19)
  n28 = Subtract(A=ComponentMask(None=TextureCoordinate() {coordinateIndex=1}) {r=true}, B=Constant() {r=2})
  n29 = Floor(None=@n28)
  n30 = Multiply(A=Saturate(None=Ceil(None=@n28)), B=OneMinus(None=Saturate(None=@n29)))
  n31 = TextureCoordinate()
  n32 = AppendVector(A=ScalarParameter() {defaultValue=-0.0381, parameterName=PanRing3}, B=Constant())
  n33 = Multiply(A=Time(), B=ScalarParameter() {defaultValue=2, parameterName=PanRing4})
  n34 = AppendVector(A=Divide(A=Ceil(None=@n33), B=ScalarParameter() {defaultValue=21, parameterName=PanRing4Fracture}), B=Constant())
  n35 = ComponentMask(None=TextureCoordinate() {coordinateIndex=1}) {r=true}
  n36 = Ceil(None=Subtract(A=@n35, B=Constant() {r=3}))
  n37 = TextureSample(UVs=LinearInterpolate(A=Panner(Coordinate=@n31, Speed=@n32), B=Add(A=@n31, B=@n34), Alpha=Saturate(None=@n36))) {samplerType=SAMPLERTYPE_Masks, texture=&T_JCount_M}
  n38 = VectorParameter() {parameterName=ColorRing3, defaultValue=<rgba 5 1.2344 0 1>}
  n39 = Subtract(A=@n35, B=Constant() {r=3})
  n40 = Floor(None=Add(A=@n39) {constB=-1})
  n41 = Multiply(A=Saturate(None=Ceil(None=@n39)), B=OneMinus(None=Saturate(None=@n40)))
  n42 = AppendVector(
   A=ScalarParameter() {group=Numbers, parameterName=u_DHMSPos}
   B=ScalarParameter() {defaultValue=-0.6, group=Numbers, parameterName=v_DHMSPos}
  )
  n43 = ScalarParameter() {group=Numbers, parameterName=u_NumberPaddingOffset}
  n44 = ScalarParameter() {defaultValue=-0.1448, group=Numbers, parameterName=v_NumberPaddingOffset}
  n45 = MaterialFunctionCall!(
   UVs=Add(A=TextureCoordinate(), B=AppendVector(A=@n43, B=@n44))
   UVPadding=AppendVector(
    A=ScalarParameter() {defaultValue=0.7867, group=Numbers, parameterName=u_NumberPaddingScale}
    B=ScalarParameter() {defaultValue=0.1543, group=Numbers, parameterName=v_NumberPaddingScale}
   )
   "NumberofDigits(int)"=ScalarParameter() {defaultValue=11, parameterName=NumberofDigits}
  ) {materialFunction=/Game/Packages/Fortress_SharedMaterials/Base_Material_Functions/BaseCharacterShaderFunctions/MF_CountdownTimer}
  n46 = ScalarParameter() {defaultValue=1.0383, group=Numbers, parameterName=u_DHMSScale}
  n47 = ScalarParameter() {defaultValue=0.419, group=Numbers, parameterName=v_DHMSScale}
  n48 = TextureSampleParameter2D(UVs=MaterialFunctionCall(UVs=Add(A=@n42, B=@n45.UVS), Texture Scale=AppendVector(A=@n46, B=@n47)) {materialFunction=&ScaleUVsByCenter}.UVs) {
   parameterName=DHMS
   samplerSource=SSM_Clamp_WorldGroupSettings
   samplerType=SAMPLERTYPE_LinearColor
   texture=/Game/Athena/Events/Generic/CountdownTimer/Textures/RT_LocalizedCountdownText
  }
  n49 = Add(A=Multiply(A=ScalarParameter() {defaultValue=1, parameterName=DHMS_Brightness}, B=@n48.R), B=@n45.NumbersGreyscale)
  LinearInterpolate(
   A=LinearInterpolate(A=LinearInterpolate(A=@n23, B=@n27, Alpha=@n30), B=Multiply(A=@n37.R, B=@n38.RGB), Alpha=@n41)
   B=Multiply(A=VectorParameter() {parameterName=Color, defaultValue=<rgba 15 6.8939 0 1>}.RGB, B=@n49)
   Alpha=@n49
  )
  n50 = VertexColor()
  n51 = CollectionParameter() {collection=&MPC_Countdown, parameterName=Digit4}
  n52 = CollectionParameter() {collection=&MPC_Countdown, parameterName=Digit3}
  n53 = Add(A=@n8, B=Multiply(A=Add(A=@n51, B=@n52), B=ScalarParameter() {defaultValue=1, parameterName=DebugProgress3}))
  n54 = Constant()
  n55 = ScalarParameter() {defaultValue=-24.373, parameterName=BackgroundStage2}
  n56 = If(A=@n3, B=Constant(), "A > B"=ScalarParameter() {defaultValue=-45.8359, parameterName=BackgroundStage1}, "A == B"=@n55, "A < B"=@n55)
  n57 = ScalarParameter() {defaultValue=-1.6288, parameterName=BackgroundStage3}
  n58 = ScalarParameter() {defaultValue=18.7986, parameterName=BackgroundStage4}
  n59 = Multiply(
   A=Transform(None=Constant3Vector() {constant=<rgba 1 0 0 0>}) {transformSourceType=TRANSFORMSOURCE_Local}
   B=If(A=@n53, B=Constant(), "A > B"=If(A=@n8, B=@n54, "A > B"=@n56, "A == B"=@n57, "A < B"=@n57), "A == B"=@n58, "A < B"=@n58)
  )
  n60 = Subtract(A=ObjectPositionWS(), B=WorldPosition().XYZ)
  n61 = Constant3Vector()
  n62 = Add(A=Multiply(A=@n13, B=Multiply(A=@n50.B, B=@n59)), B=Multiply(A=LinearInterpolate(A=@n60, B=@n61, Alpha=@n10), B=VertexColor().R))
  n63 = Subtract(A=ObjectPositionWS(), B=Add(A=@n62, B=WorldPosition() {worldPositionShaderOffset=WPT_ExcludeAllShaderOffsets}.XYZ))
  n64 = Subtract(A=Distance(A=ObjectPositionWS(), B=CameraPositionWS()), B=ScalarParameter() {defaultValue=500, parameterName=TurnOffDistance})
  n65 = ScalarParameter() {defaultValue=100, parameterName=TurnOffFade}
  n66 = ComponentMask(None=Saturate(None=Divide(A=@n64, B=@n65))) {r=true}
  n67 = Max(
   A=ScalarParameter() {parameterName=ForceOn}
   B=CollectionParameter() {collection=&MPC_Countdown, parameterName=ShowCountdown}
  )
  LinearInterpolate(A=@n63, B=LinearInterpolate(A=@n63, B=@n62, Alpha=Ceil(None=@n66)), Alpha=@n67)
  n68 = TextureCoordinate()
  n69 = Multiply(
   A=Add(
    A=CollectionParameter() {collection=&MPC_Countdown, parameterName=Digit2}
    B=CollectionParameter() {collection=&MPC_Countdown, parameterName=Digit1}
   )
   B=ScalarParameter() {defaultValue=1, parameterName=DebugProgress4}
  )
  n70 = Constant()
  n71 = ScalarParameter() {defaultValue=0.2533, parameterName=ProgresBarStage2}
  n72 = ScalarParameter() {defaultValue=0.5488, parameterName=ProgresBarStage3}
  n73 = If(
   A=@n8
   B=Constant()
   "A > B"=If(A=@n3, B=Constant(), "A > B"=ScalarParameter() {defaultValue=0.099, parameterName=ProgresBarStage1}, "A == B"=@n71, "A < B"=@n71)
   "A == B"=@n72
   "A < B"=@n72
  )
  n74 = ScalarParameter() {defaultValue=0.8781, parameterName=ProgresBarStage4}
  n75 = ScalarParameter() {defaultValue=1, parameterName=ProgresBarStage5}
  n76 = Add(
   A=OneMinus(None=ComponentMask(None=@n68) {r=true})
   B=If(A=Add(A=@n53, B=@n69), B=@n70, "A > B"=If(A=@n53, B=@n70, "A > B"=@n73, "A == B"=@n74, "A < B"=@n74), "A == B"=@n75, "A < B"=@n75)
  )
  n77 = Multiply(A=@n15.R, B=Saturate(None=Floor(None=@n76)))
  n78 = Multiply(
   A=LinearInterpolate(A=@n15.R, B=Add(A=@n77, B=@n15.G), Alpha=VertexColor().G)
   B=LinearInterpolate(
    A=ScalarParameter() {defaultValue=0.381, parameterName=OpaciytBackground}
    B=ScalarParameter() {defaultValue=5, parameterName=OpaciytBackgroundPlate}
    Alpha=@n13
   )
  )
  n79 = Multiply(A=ScalarParameter() {defaultValue=0.101, parameterName=OpacityRing1}, B=@n18.R)
  n80 = ScalarParameter() {defaultValue=0.101, parameterName=OpacityRing2}
  n81 = LinearInterpolate(
   A=LinearInterpolate(A=LinearInterpolate(A=@n78, B=@n79, Alpha=@n22), B=Multiply(A=@n80, B=@n26.R), Alpha=@n30)
   B=Multiply(A=ScalarParameter() {defaultValue=0.101, parameterName=OpacityRing3}, B=@n37.R)
   Alpha=@n41
  )
  LinearInterpolate(A=Constant(), B=LinearInterpolate(A=Constant3Vector(), B=Add(A=@n81, B=@n49), Alpha=@n66), Alpha=@n67)
  n82 = ObjectPositionWS()
  n83 = WorldPosition()
  n84 = Multiply(
   A=@n67
   B=CollectionParameter() {collection=&MPC_Countdown, parameterName=ShowCountdownPhase2}
  )
  LinearInterpolate(
   A=Add(A=Subtract(A=@n82, B=@n83.XYZ), B=Constant3Vector() {constant=<rgba 0 0 -5156156 0>})
   B=Constant3Vector()
   Alpha=StaticSwitchParameter(True=Constant() {r=1}, False=StaticSwitchParameter(True=@n84, False=@n67) {parameterName=isPhase2}) {parameterName=Debug}
  )
 ]
}
